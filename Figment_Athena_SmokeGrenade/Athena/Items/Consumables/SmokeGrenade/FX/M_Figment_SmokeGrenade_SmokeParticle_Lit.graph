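# Translucent, volumetric-lit Niagara sprite material for smoke grenade particles.
# Dynamic parameters drive 8x8 flipbooks for depth, alpha and normals, plus a scrolling, particle-randomized noise for dissolve detail.
# Alpha combines contrast, depth fade and particle alpha, with switchable near-camera, Z-axis mask and sunlight-based shadow compensation.
# Flipbook normals blend with a sphere normal and optional directional bias, while color is shaded by sky-light-to-camera alignment.
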Material /Figment_Athena_SmokeGrenade/Athena/Items/Consumables/SmokeGrenade/FX/M_Figment_SmokeGrenade_SmokeParticle_Lit
graph {
 properties {
  bUsedWithNanite=true
  bUsedWithNiagaraSprites=true
  blendMode=BLEND_Translucent
  pixelDepthOffsetMode=PDOM_Legacy
  translucencyLightingMode=TLM_VolumetricPerVertexDirectional
  usedShadingModels=Default Lit
 }
 nodes [
  n1 = DotProduct(A=SkyAtmosphereLightDirection(), B=CameraVectorWS())
  n2 = NamedRerouteUsage()
  n3 = ScalarParameter() {defaultValue=0.2, parameterName=Shade_Lerp}
  Multiply(
   A=LinearInterpolate(B=Saturate(None=@n1), Alpha=ScalarParameter() {defaultValue=0.4, parameterName=Dot Blend}) {constA=1}
   B=Multiply(A=ParticleColor().RGB, B=LinearInterpolate(B=@n2, Alpha=@n3) {constA=1})
  )
  n4 = DynamicParameter() {paramNames [Animation, Unused, Emissive, NormalMultiplier]}
  n5 = Constant() {r=8}
  n6 = MaterialFunctionCall(
   "Animation  Phase (0-1)"=@n4.Animation
   Number of Rows=@n5
   Number of Columns=@n5
   Texture=TextureObjectParameter() {parameterName=Texture, texture=/Game/Effects/Fort_Effects/Textures/Smoke/T_Smoke_Thick_8x8_02}
  ) {materialFunction=&FlipBook}
  NamedRerouteDeclaration(None=StaticComponentMaskParameter(None=@n6.Result) {defaultG=true, parameterName=Depth Channel Mask}) {name=Depth_Flipbook, nodeColor=<rgba 0 0.8594 1 1>}
  NamedRerouteDeclaration(None=@n6.Alpha) {name=Alpha_Flipbook, nodeColor=<rgba 1 0 0.3516 1>}
  n7 = ComponentMask(None=NamedRerouteUsage()) {g=true, r=true}
  n8 = DynamicParameter() {
   parameterIndex=2
   defaultValue=<rgba 0 0 0 1>
   paramNames [Smoke - DissolveLerpNoise, Smoke - DissolveNoiseOffsetX, Smoke - DissolveNoiseErosionPower, Smoke - DissolveNoiseScale]
  }
  n9 = Constant()
  n10 = ScalarParameter() {group=Normal, parameterName="DissolveDetail  Random Y Offset Value", sliderMax=2}
  n11 = Add(A=Add(A=TextureCoordinate(), B=@n8.Smoke - DissolveNoiseOffsetX), B=Multiply(A=ParticleRandom(), B=AppendVector(A=@n9, B=@n10)))
  n12 = TextureSampleParameter2D(UVs=Add(A=Multiply(A=@n7) {constB=-0.2}, B=Multiply(A=@n11, B=@n8.Smoke - DissolveNoiseScale) {desc=Scale})) {parameterName=DissolveDetailTexture, samplerType=SAMPLERTYPE_LinearColor, texture=/Game/Global/Textures/Default/Noise/T_Noise01_Inverse}
  n13 = DynamicParameter() {
   parameterIndex=2
   defaultValue=<rgba 0 0 0 1>
   paramNames [Smoke - DissolveLerpNoise, Smoke - DissolveNoiseOffsetX, Smoke - DissolveNoiseErosionPower, Smoke - DissolveNoiseScale]
  }
  NamedRerouteDeclaration(None=Clamp(None=Power(Base=@n12.R, Exp=@n13.Smoke - DissolveNoiseErosionPower))) {name=DissolveDetail, nodeColor=<rgba 0.3828 0 1 1>}
  n14 = ScalarParameter() {defaultValue=100, parameterName=DepthFade_Low}
  n15 = ScalarParameter() {defaultValue=10, parameterName=DepthFade_high}
  n16 = Multiply(A=NamedRerouteUsage(), B=NamedRerouteUsage())
  NamedRerouteDeclaration(None=DepthFade(FadeDistance=LinearInterpolate(A=@n14, B=@n15, Alpha=@n16))) {name=DepthFade, nodeColor=<rgba 0.875 1 0 1>}
  n17 = Subtract(A=NamedRerouteUsage(), B=Constant() {r=-1})
  n18 = ScalarParameter() {parameterName=Contrast}
  n19 = NamedRerouteUsage()
  n20 = NamedRerouteUsage()
  n21 = Multiply(
   A=Multiply(
    A=MaterialFunctionCall(In=@n17, Contrast=@n18) {materialFunction=&CheapContrast}.Result
    B=Multiply(A=@n19, B=@n20)
   )
   B=ParticleColor().A
  )
  n22 = SphereMask(
   A=PixelDepth().R
   B=Constant()
   Radius=ScalarParameter() {defaultValue=400, parameterName=NearCameraFadeRadius}
   Hardness=ScalarParameter() {parameterName=NearCameraFadeHardness}
  )
  n23 = StaticSwitchParameter(True=Multiply(A=@n21, B=OneMinus(None=@n22)), False=@n21) {parameterName=IsUsingCameraFade}
  n24 = ParticlePositionWS()
  n25 = VectorParameter() {parameterName=ParticleZaxisOpacityMask_ParticlePositionOffset, defaultValue=<rgba 0 0 -15 1>}
  n26 = ComponentMask(None=Subtract(A=Add(A=@n24, B=@n25.RGB), B=WorldPosition().XYZ)) {b=true}
  n27 = Saturate(None=Divide(A=Multiply(A=@n26) {constB=-1}, B=ScalarParameter() {defaultValue=20, parameterName=ParticleZaxisOpacityMask_Blur}))
  NamedRerouteDeclaration(None=StaticSwitchParameter(True=Multiply(A=@n23, B=@n27), False=@n23) {parameterName=ParticleZaxisOpacityMask}) {name=Final Alpha, nodeColor=<rgba 1 0 0.2578 1>}
  n28 = TextureObjectParameter() {parameterName=NormalTexture, samplerType=SAMPLERTYPE_Normal, texture=/Game/Effects/Fort_Effects/Textures/Smoke/T_FallingMist_8x8_01_N}
  NamedRerouteDeclaration(None=ComponentMask(None=MaterialFunctionCall("Animation  Phase (0-1)"=@n4.Animation, Number of Rows=@n5, Number of Columns=@n5, Texture=@n28) {materialFunction=&FlipBook}.Result) {b=true, g=true, r=true}) {name=Normal_Flipbook, nodeColor=<rgba 0.9453 0 1 1>}
  n29 = Constant3Vector() {constant=<rgba 0 0 1 0>}
  n30 = NamedRerouteDeclaration(None=TextureSample() {samplerType=SAMPLERTYPE_Normal, texture=/Engine/Functions/Engine_MaterialFunctions02/ExampleContent/Textures/SphereRenderNormalsMap}.RGB) {name=SphereNormal_Texture, nodeColor=<rgba 1 0.3281 0 1>}
  n31 = Multiply(
   A=ScalarParameter() {defaultValue=1, group=Normal, parameterName=SphereNormal}
   B=DynamicParameter() {paramNames [Animation, Unused, Emissive, NormalMultiplier]}.NormalMultiplier
  )
  NamedRerouteDeclaration(None=Normalize(VectorInput=LinearInterpolate(A=@n29, B=@n30, Alpha=@n31))) {name=SphereNormal, nodeColor=<rgba 1 0 0.2109 1>}
  n32 = ScalarParameter() {defaultValue=90, desc=FOV Scale Floor, group=CamFade, parameterName=CamFade FOV Floor, sortPriority=4}
  n33 = Max(A=MaterialFunctionCall() {materialFunction=/Engine/Functions/Engine_MaterialFunctions02/Utility/FOV}.FOV angle in degrees, B=@n32)
  n34 = ScalarParameter() {defaultValue=130, desc=FOV Scale Celing, group=CamFade, parameterName=CamFade FOV Ceil, sortPriority=5}
  n35 = ScalarParameter() {
   defaultValue=1
   desc="Scales INPUT DISTANCE by FOV (before remapping it to near/far) - so larger numbers reduce the effect. Floor defaults to 1.0"
   group=CamFade
   parameterName=CamFade ScaleDistByFOV Floor
   sortPriority=8
  }
  n36 = ScalarParameter() {
   defaultValue=1
   desc="Scales INPUT DISTANCE by FOV (before remapping it to near/far) - so larger numbers reduce the effect. Celing defaults to 1.0 but is meaningful at 2.0"
   group=CamFade
   parameterName=CamFade ScaleDistByFOV Ceiling
   sortPriority=9
  }
  n37 = CameraPositionWS()
  n38 = ParticlePositionWS()
  n39 = Multiply(
   A=MaterialFunctionCall(Input=Min(A=@n33, B=@n34), Input Low=@n32, Input High=@n34, Target Low=@n35, Target High=@n36) {materialFunction=&RemapValueRange}.Result
   B=Length(None=Subtract(A=@n37, B=@n38)) {is=class MaterialExpressionLength}
  )
  n40 = ScalarParameter() {defaultValue=25, desc=Near distance to camera --250, group=CamFade, parameterName=CamFade Near, sliderMax=250, sortPriority=1}
  n41 = ScalarParameter() {defaultValue=25, desc=Far distance to camera -- 300, group=CamFade, parameterName=CamFade Far, sliderMax=300, sortPriority=2}
  NamedRerouteDeclaration(None=Saturate(None=MaterialFunctionCall(Input=@n39, Input Low=@n40, Input High=@n41) {materialFunction=&RemapValue_0-1}.Result)) {name=ClipFade Alpha, nodeColor=<rgba 0 1 0.9447 1>}
  n42 = NamedRerouteUsage()
  n43 = ComponentMask(None=VectorParameter() {group=Normal, parameterName=NormalBias Direction, defaultValue=<rgba 0.5 0.5 0 0>}.RGB) {b=true, g=true, r=true}
  n44 = ScalarParameter() {defaultValue=1, desc=Toggle the Bias direction from Tangent to World Space, group=Normal, parameterName=NormalBias WorldSpace}
  n45 = Add(
   A=@n42
   B=LinearInterpolate(A=@n43, B=Transform(None=@n43) {transformSourceType=TRANSFORMSOURCE_World, transformType=TRANSFORM_Tangent}, Alpha=@n44)
  )
  n46 = ComponentMask(None=VectorParameter() {group=Normal, parameterName=NormalBias Multiplier, defaultValue=<rgba 2 2 1 0>}.RGB) {b=true, g=true, r=true}
  NamedRerouteDeclaration(None=StaticSwitchParameter(True=@n42, False=Multiply(A=@n45, B=@n46)) {defaultValue=true, group=Normal, parameterName=DontUseNormalBias}) {name=BiasedNormal, nodeColor=<rgba 0.6875 1 0 1>}
  n47 = LinearInterpolate(A=NamedRerouteUsage(), B=NamedRerouteUsage(), Alpha=ScalarParameter() {defaultValue=1, group=Normal, parameterName=NormalTexture})
  NamedRerouteDeclaration(None=MaterialFunctionCall(BaseNormal=Normalize(VectorInput=@n47), AdditionalNormal=NamedRerouteUsage()) {materialFunction=&BlendAngleCorrectedNormals}.Result) {name=Final Normal, nodeColor=<rgba 1 0 0.75 1>}
  NamedRerouteUsage()
  n48 = Distance(A=ObjectPositionWS(), B=CameraPositionWS())
  n49 = ScalarParameter() {defaultValue=299, parameterName=Radius}
  n50 = Constant() {r=1}
  n51 = Constant()
  n52 = Subtract(A=WorldPosition().XYZ, B=ParticlePositionWS())
  n53 = ScalarParameter() {defaultValue=50, parameterName=Particle Scale Control}
  StaticSwitchParameter(True=Multiply(A=If(A=@n48, B=@n49, "A > B"=@n50, "A == B"=@n50, "A < B"=@n51), B=Multiply(A=@n52, B=@n53)), False=Constant()) {parameterName=Use_Near Camera Particle Scale 0}
  n54 = Constant3Vector()
  n55 = CollectionParameter() {
   collection=&FortniteMaterialParameters
   parameterName=SunlightDirectionalVector
  }
  n56 = NamedRerouteUsage()
  n57 = DotProduct(
   A=ComponentMask(None=@n55) {b=true, g=true, r=true}
   B=MaterialFunctionCall(Normal=Transform(None=@n56), Flatness=Constant() {r=0.7}) {materialFunction=&FlattenNormal}.Result
  )
  n58 = LinearInterpolate(A=ScalarParameter() {defaultValue=2, parameterName=ShadowAlphaMult}, Alpha=Saturate(None=Multiply(A=@n57) {constB=5}))
  n59 = NamedRerouteUsage()
  n60 = NamedRerouteUsage()
  StaticSwitchParameter(True=Multiply(A=LinearInterpolate(A=@n54, B=@n58, Alpha=@n59), B=@n60), False=@n60) {parameterName=ShadowAlphaCompensation}
 ]
}
