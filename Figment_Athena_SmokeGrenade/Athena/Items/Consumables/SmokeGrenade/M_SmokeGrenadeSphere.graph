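# Translucent material for the smoke grenade's sphere, giving an animated smoke color and an opacity tied to camera proximity.
# Two panning samples of one noise texture modulate the base color, weighted by a squared vertical sine band.
# A camera-within-radius mask gates a soft world-space sphere mask for opacity and also scales a camera-to-object offset vector.
# Depth fade softens intersections on default and deferred paths, while mobile skips it.
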Material /Figment_Athena_SmokeGrenade/Athena/Items/Consumables/SmokeGrenade/M_SmokeGrenadeSphere
graph {
 properties {blendMode=BLEND_Translucent, ditherOpacityMask=true, pixelDepthOffsetMode=PDOM_Legacy, usedShadingModels=Default Lit}
 nodes [
  Constant3Vector() {constant=<rgba 1 0 0.0317 1>}
  n1 = VectorParameter() {parameterName=BaseColorLow, defaultValue=<rgba 0.18 0.23 0.3 0>}
  MaterialFunctionCall(
   Color_Morning=@n1.RGB
   Color_Day=@n1.RGB
   Color_Evening=@n1.RGB
   Color_Night=VectorParameter() {parameterName=BaseColorLow_Night, defaultValue=<rgba 0.1 0.1 0.1 0>}.RGB
  ) {materialFunction=&Color_Time_of_Day}
  n2 = Multiply(A=@n1.RGB) {constB=1.3}
  n3 = ScalarParameter() {defaultValue=1, parameterName=U1}
  n4 = ScalarParameter() {defaultValue=1.3, parameterName=V1}
  n5 = Panner(Coordinate=Multiply(A=TextureCoordinate(), B=AppendVector(A=@n3, B=@n4))) {bFractionalPart=true, speedY=0.02}
  n6 = ScalarParameter() {defaultValue=2, parameterName=U2}
  n7 = ScalarParameter() {defaultValue=1.125, parameterName=V2}
  n8 = Panner(Coordinate=Multiply(A=TextureCoordinate(), B=AppendVector(A=@n6, B=@n7))) {bFractionalPart=true, speedY=0.0672}
  n9 = Divide(A=Add(
    A=TextureSample(UVs=@n5) {texture=&2ndInfectedMapLinear}.RGB
    B=TextureSample(UVs=@n8) {texture=&2ndInfectedMapLinear}.RGB
   ))
  n10 = Sine(None=ComponentMask(None=TextureCoordinate()) {g=true}) {period=2}
  n11 = LinearInterpolate(A=@n1.RGB, B=LinearInterpolate(A=@n1.RGB, Alpha=Multiply(A=@n2, B=@n9)) {constA=0.5}, Alpha=Multiply(A=@n10, B=@n10))
  MaterialFunctionCall(Emissive=@n11) {materialFunction=&MF_FX_EmissiveAdjustments}
  MaterialFunctionCall(Emissive=Multiply(A=@n11, B=ScalarParameter() {defaultValue=0.13}) {constB=0.15}) {materialFunction=&MF_FX_EmissiveAdjustments}
  n12 = CameraPositionWS()
  n13 = ObjectPositionWS()
  n14 = ScalarParameter() {defaultValue=1250, parameterName=Radius}
  n15 = SphereMask(A=@n12, B=@n13, Radius=@n14)
  Multiply(A=Subtract(A=@n12, B=@n13), B=@n15)
  n16 = WorldPosition()
  n17 = ObjectPositionWS()
  n18 = Multiply(A=ScalarParameter() {defaultValue=0.78, parameterName=Ratio}, B=@n14)
  n19 = Multiply(A=ScalarParameter() {parameterName=Opacity}, B=Multiply(A=SphereMask(A=@n16.XYZ, B=@n17, Radius=@n18) {hardnessPercent=80}, B=@n15))
  n20 = DepthFade(Opacity=@n19, FadeDistance=ScalarParameter() {defaultValue=20, parameterName=FadeDistance})
  ShadingPathSwitch(Default=@n20, Deferred=@n20, Mobile=@n19) {desc=Migrated from FeatureLevelSwitch}
 ]
}
